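; Minor spire boss gameplay ability that fires a charged laser at sensed player pawns in shuffled order, up to an activation limit.
; Targets pass a line trace requiring a damageable hit, then timers drive turning, charge, release and recovery alongside montage and gameplay cues.
; A delayed snapshot of the target's foot position fixes the aim before firing.
; The beam traces along that direction and triggers an explosion on hit.
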(event Ability|EventActivateAbility
  ; Activate Ability
  (|CallParentFunction)
  (Default|InitVariables)
  (bind _returnvalue (Ability|CommitAbility))
  (if _returnvalue
    (bind _returnvalue_2 (Ability|ApplyGameplayEffectToOwner (Variables|Default|GetGE_AOEAbilityActive)))
    (Variables|Default|SetAbilityActiveHandle _returnvalue_2)
    (CallFunction|TrytoAttackCurrentTarget)))

(event Custom|TryToAttackCurrentTarget
  (bind _currenttargetindex (Variables|Default|GetCurrentTargetIndex))
  (bind _returnvalue (Math|Float|Truncate _currenttargetindex))
  (bind _spirebga (Variables|Default|GetSpireBGA))
  (bind _sensedplayerpawns (Variables|/SpireBoss/Gameplay/MinorSpire/BGA_Spire_Minor.BGA_Spire_Minor_C|GetSensedPlayerPawns _spirebga))
  (bind _output (Utilities|Array|Get(acopy) _sensedplayerpawns _returnvalue))
  ; Check LOS towards current sensed target / if they are behind a damageable object. If they are valid, start facing them and start charging
  (Development|PrintString (Utilities|String|ToString(Text) (Utilities|Text|FormatText "NSLOCTEXT(\"[712E1C3583B18130D33D1718F32D2A3A]\", \"79954EAC4E682D5FD8753D9405B85226\", \"TryToAttackCurrentTarget: {TargetIndex}, Target Name: {TargetName}\")" (Variables|Default|GetCurrentTargetIndex) (Fortnite|GetPlayerNameSafe _output))) false (Variables|Debug|GetbVerboseLogging))
  (bind _bisviabletarget (CallFunction|IsViableTarget _output))
  (if _bisviabletarget
    (bind _output_get (Variables|Default|SetCurrentLaserTarget _output))
    (Class|BGASpireMinor|StartTargetFacing (Variables|Default|GetSpireBGA) _output_get)
    (bind _returnvalue_2 (Utilities|Time|SetTimerbyEvent (EventDispatchers|CreateEvent :!event Custom|ChargeAttack) (Variables|Default|GetTurnToFaceTime)))
    (Variables|Default|SetChargeAttackHandle _returnvalue_2)
    (else
      (CallFunction|ChooseNextTarget))))

(event Custom|ChargeAttack
  ; Charge Attack
  (Development|PrintString "Starting Attack Charge" false (Variables|Debug|GetbVerboseLogging))
  (Class|BGASpireMinorAttacking|StartLazerCharge (Variables|Default|GetSpireBGA))
  (GameplayCue|AddGameplayCueOnActor(Looping) (Variables|Default|GetSpireBGA) (Variables|GameplayCueTags|GetCharge%20Tag) (Ability|GameplayCue|MakeGameplayCueParameters 0.0)) ; Charge Bean GC
  (bind _returnvalue (Utilities|Time|SetTimerbyEvent (EventDispatchers|CreateEvent :!event Custom|ReleaseAttack) (Variables|Default|GetChargeDuration))) ; Wait for charge to release that laser
  (Variables|Default|SetReleaseAtkHandle _returnvalue)
  (bind _asynctaskproxy (Ability|Tasks|PlayMontageAndWaitNotify "MinorSpire_Laser" "/SpireBoss/Assets/Animation/SpireMinor/SpireMinor_Beam_Attack_M.SpireMinor_Beam_Attack_M" 1.0 "Intro")
    (:then
      (Variables|Default|SetAttackTask _asynctaskproxy)
      (CallFunction|TriggerDelayedTargeting))
    (:OnCompleted
      _)
    (:OnInterrupted
      (CallFunction|HandleInterrupted))
    (:OnCancelled
      _))
  (Ability|EndAbility))

(event Custom|SnapshotLastSeenDirection
  (bind _currentlasertarget (Variables|Default|GetCurrentLaserTarget))
  (bind (_returnvalue_x _returnvalue_y _returnvalue_z) (Math|Vector|BreakVector (Transformation|GetActorLocation _currentlasertarget)))
  (bind _capsulecomponent (Variables|/Script/Engine.Character|GetCapsuleComponent _currentlasertarget))
  (bind _returnvalue (Components|Capsule|GetScaledCapsuleHalfHeight _capsulecomponent))
  (bind _returnvalue_2 (- _returnvalue_z _returnvalue))
  (bind _returnvalue_3 (Math|Vector|MakeVector _returnvalue_x _returnvalue_y _returnvalue_2))
  ; Set Laser Target Location
  (Debug|DebugLastSeenLocation _returnvalue_3)
  (bind _output_get (Variables|/SpireBoss/Gameplay/MinorSpire/BGA_Spire_Minor_Attacking.BGA_Spire_Minor_Attacking_C|SetLaserTargetDirection (Math|Vector|Normalize (Math|Vector|GetForwardVector (Math|Rotator|FindLookatRotation (CallFunction|GetAttackOriginSocket) _returnvalue_3))) (Variables|Default|GetSpireBGA)))
  (Development|PrintString (Utilities|String|ToString(Text) (Utilities|Text|FormatText "NSLOCTEXT(\"[712E1C3583B18130D33D1718F32D2A3A]\", \"CFCB0E964F97C0F8CB9631BDD12BA337\", \"LastSeenTargetDirection Recorded: {Location}\")" (Utilities|String|ToString(Vector) _output_get))) false (Variables|Debug|GetbVerboseLogging)))

(event Custom|ReleaseAttack
  ; Release Attack - Target and Damage things in range of the explosion
  (bind _outhit (Collision|LineTraceByChannel (CallFunction|GetAttackOriginSocket) (+ (CallFunction|GetAttackOriginSocket) (* (Variables|/SpireBoss/Gameplay/MinorSpire/BGA_Spire_Minor_Attacking.BGA_Spire_Minor_Attacking_C|GetLaserTargetDirection (Variables|Default|GetSpireBGA)) (ScalableFloat|GetValueatLevel (Variables|Default|GetHF_LaserMaxRange)))) "TraceTypeQuery5" true (Variables|Default|GetIgnoredActors))) ; We target the laser pointed at the direction to the last known target location
  (GameplayCue|AddGameplayCueOnActor(Looping) (Variables|Default|GetSpireBGA) (Variables|GameplayCueTags|GetBeam%20Tag) (Ability|GameplayCue|MakeGameplayCueParameters :NormalizedMagnitude 0.0 :RawMagnitude (Variables|Default|GetBeamDuration) :Location (select (Collision|BreakHitResult _outhit) (.Location (Collision|BreakHitResult _outhit)) (.TraceEnd (Collision|BreakHitResult _outhit))) :Normal "0,0,0" :Instigator (Variables|Default|GetSpireBGA) :GameplayEffectLevel 1 :AbilityLevel 1 :bReplicateLocationWhenUsingMinimalRepProxy false)) ; Play Laser Cue with the target location \nsaved previously
  (if (Collision|BreakHitResult _outhit)
    (Utilities|TriggerLaserExplosion _outhit)
    (else
      _))
  (CallFunction|TriggerLaserRecovery))

(event Custom|TriggerLaserRecovery
  ; Recovery - Play Outro and wait for Reactivation Cooldown and Beam effect to finish
  (Ability|Animation|MontageJumptoSection "Outro")
  (bind _returnvalue (Utilities|Time|SetTimerbyEvent (EventDispatchers|CreateEvent :!event Custom|ChooseNextTarget) (+ (Variables|Default|GetBeamDuration) (ScalableFloat|GetValueatLevel (Variables|Default|GetDelayBetweenReactivations)))))
  (Variables|Default|SetReactivationHandle _returnvalue))

(event Custom|ChooseNextTarget
  ; Reactivate only if we shot at all of our targets or we got to the limit of our shots per ability activation
  (bind _result (Math|Float|IncrementFloat (Variables|Default|GetCurrentTargetIndex))
    (:%20%20
      (if (< _result (Math|Float|Min(Float) (Math|Conversions|ToFloat(Integer) (Utilities|Array|Length (Variables|/SpireBoss/Gameplay/MinorSpire/BGA_Spire_Minor_Attacking.BGA_Spire_Minor_Attacking_C|GetSensedPlayerPawns (Variables|Default|GetSpireBGA)))) (ScalableFloat|GetValueatLevel (Variables|Default|GetMaximumLazerActivation))))
        (CallFunction|ClearCues)
        (CallFunction|TrytoAttackCurrentTarget)
        (else
          (Development|PrintString "ChooseNextTarget: Finished All Targets. Ending Ability" false (Variables|Debug|GetbVerboseLogging))
          (Ability|EndAbility))))))

(event Custom|ClearCues
  ; Clear Gameplay Cues
  (for _array_element (Utilities|Array|MakeArray (Variables|GameplayCueTags|GetBeam%20Tag) (Variables|GameplayCueTags|GetCharge%20Tag))
    (GameplayCue|RemoveGameplayCueOnActor(Looping) (Variables|Default|GetSpireBGA) _array_element (Ability|GameplayCue|MakeGameplayCueParameters 0.0))))

(event Ability|EventOnEndAbility
  (params (bWasCancelled :type Boolean))
  ; End Ability - Clear timers, effects and cues
  (|CallParentFunction bWasCancelled)
  (CallFunction|ClearCues)
  (Ability|RemoveGameplayEffectFromOwnerWithHandle (Variables|Default|GetAbilityActiveHandle) 1)
  (Class|BGASpireMinorAttacking|ResetLazerCharge (Variables|Default|GetSpireBGA))
  (for _array_element (Utilities|Array|MakeArray (Variables|Default|GetReleaseAtkHandle) (Variables|Default|GetReactivationHandle) (Variables|Default|GetTargetCueHandle) (Variables|Default|GetChargeAttackHandle) (Variables|Default|GetTargetDelayedHandle))
    (Utilities|Time|ClearandInvalidateTimerbyHandle _array_element)))

(fn ApplyLaserExplosionEffects
  (params (TargetData :type Gameplay%20Ability%20Target%20Data%20Handle%20Structure%20%28by%20ref%29) (ApplicationTag :type Gameplay%20Tag%20Structure))
  (Ability|ApplyGameplayEffectContainer TargetData ApplicationTag)
  (Debug|PrintActorArray (Ability|TargetData|GetActorsfromTargetData TargetData) "Filtered Targets:"))

(fn CheckCanBeDamaged
  (params (Actor :type Actor%20Object%20Reference))
  (Utilities|IsValid Actor
    (:Is%20Valid
      (Variables|Default|SetTarget%20Can%20Be%20Damaged (Variables|/Script/Engine.Actor|GetbCanBeDamaged Actor))
      (return (Variables|Default|GetTarget%20Can%20Be%20Damaged)))
    (:Is%20Not%20Valid
      (return false))))

(fn DebugLastSeenLocation
  (params (Position :type Vector))
  (if (Variables|Debug|GetbVerboseLogging)
    (Rendering|Debug|DrawDebugSphere Position 3e+01 12 "(R=0.011253,G=0.000000,B=1.000000,A=1.000000)" 3.0)))

(fn GetAttackOriginSocket
  (return (Transformation|GetSocketLocation (Variables|/SpireBoss/Gameplay/MinorSpire/BGA_Spire_Minor.BGA_Spire_Minor_C|GetSkeletalMeshComponentBudgeted (Variables|Default|GetSpireBGA)) "masksocket")))

(fn GetCustomAbilitySourceTransform
  (bind (_aoeorigin_x _aoeorigin_y _aoeorigin_z) (Math|Vector|BreakVector (Variables|GetAOEOrigin_X)))
  (return _aoeorigin_x _aoeorigin_y (+ _aoeorigin_z 15.0) "0, 0, 0" "1.000000,1.000000,1.000000"))

(fn HandleInterrupted
  (if (not (Utilities|Time|IsTimerActivebyHandle (Variables|Default|GetReleaseAtkHandle)))
    (Ability|EndAbility)))

(fn InitVariables
  (bind _spirebga (Variables|Default|GetSpireBGA))
  (Utilities|FlowControl|Sequence
    (:then_0
      (bind _asbga_spire_minor_attacking (Utilities|Casting|CastToBGA_Spire_Minor_Attacking (Ability|GetAvatarActorfromActorInfo)))
      (Variables|Default|SetSpireBGA _asbga_spire_minor_attacking)
      (Variables|Default|SetCurrentTargetIndex 0.0)
      (Utilities|Array|Shuffle (Variables|/SpireBoss/Gameplay/MinorSpire/BGA_Spire_Minor.BGA_Spire_Minor_C|GetSensedPlayerPawns (Variables|Default|GetSpireBGA)))
      (Variables|Default|SetCurrentLaserTarget 0))
    (:then_1
      (Utilities|Array|AddUnique (Variables|Default|GetIgnoredActors) _spirebga)
      (Utilities|Array|AddUnique (Variables|Default|GetIgnoredActors) (Variables|/Script/Engine.ChildActorComponent|GetChildActor (Variables|/SpireBoss/Gameplay/MinorSpire/BGA_Spire_Minor.BGA_Spire_Minor_C|GetWeakpoint01 _spirebga))))))

(fn IsViableTarget
  (params (PawnToCheck :type Fort%20Player%20Pawn%20Object%20Reference))
  (bind _outhit (Collision|LineTraceByChannel (CallFunction|GetAttackOriginSocket) (Transformation|GetActorLocation PawnToCheck) "TraceTypeQuery5" false (Variables|Default|GetIgnoredActors) "None" true "(R=1.000000,G=0.000000,B=0.000000,A=1.000000)" "(R=0.000000,G=1.000000,B=0.000000,A=1.000000)" 4.0)) ; Line of sight check
  (Utilities|IsValid (.HitActor (Collision|BreakHitResult _outhit))
    (:Is%20Valid
      (LocalVariables|SetbIsValidTargetResult (Variables|/Script/Engine.Actor|GetbCanBeDamaged (.HitActor (Collision|BreakHitResult _outhit)))))
    (:Is%20Not%20Valid
      (LocalVariables|SetbIsValidTargetResult false))) ; As long as we hit something damageable \non the way to the player, we are aa viable target.\nYou can hide behind non damageable things and the\nspire won't target you
  (Development|PrintString (Utilities|String|ToString(Text) (Utilities|Text|FormatText "NSLOCTEXT(\"[712E1C3583B18130D33D1718F32D2A3A]\", \"9685EB1D4E14E3AAC505D2B7CE34846D\", \"IsViableTarget: Actor: {Actor} Hit Valid Actor: {ValidActor} CanBeDamaged: {CanBeDamaged}\")" (Fortnite|GetPlayerNameSafe (Variables|GetPawnToCheck)) (Utilities|IsValid (.HitActor (Collision|BreakHitResult _outhit)) :!class /Script/Engine.KismetSystemLibrary) (Variables|/Script/Engine.Actor|GetbCanBeDamaged (.HitActor (Collision|BreakHitResult _outhit))))) false (Variables|Debug|GetbVerboseLogging))
  (return (LocalVariables|GetbIsValidTargetResult)))

(fn PrintActorArray
  (params (Actors :type Array%20of%20Actor%20Object%20References) (Title :type String))
  (if (Variables|Debug|GetbVerboseLogging)
    (for _array_element Actors
      (LocalVariables|SetFoundTargetsString (Utilities|String|Append (LocalVariables|GetFoundTargetsString) ", " (Utilities|GetDisplayName _array_element))))
    (Development|PrintString (Utilities|String|ToString(Text) (Utilities|Text|FormatText "NSLOCTEXT(\"[712E1C3583B18130D33D1718F32D2A3A]\", \"FDF253F544B85997177D11B61E2F4073\", \"{Title}: {Targets}\")" Title (LocalVariables|GetFoundTargetsString))) false)))

(fn ShippingLogSuccess
  (params (Object :type Object%20Reference) (InInt :type Integer))
  (Fortnite|FortShippingLog self (Utilities|String|Append "SUCCESS - " "Current Activation Count : " (Utilities|String|ToString(Integer) InInt) " Target Pawn : " (Utilities|GetDisplayName Object))))

(fn TargetCue
  ; Trigger Cue that highlights the current spire laser target
  (GameplayCue|ExecuteGameplayCueOnActor(Burst) (Variables|Default|GetCurrentLaserTarget) (Variables|GameplayCueTags|GetPlayerTargeted%20Tag) (Ability|GameplayCue|MakeGameplayCueParameters :NormalizedMagnitude 0.0 :RawMagnitude 2.0 :Location "0,0,0" :Normal (Math|Vector|MakeVector 1.0) :SourceObject (Variables|/SpireBoss/Gameplay/MinorSpire/BGA_Spire_Minor.BGA_Spire_Minor_C|GetSkeletalMeshComponentBudgeted (Variables|Default|GetSpireBGA)) :GameplayEffectLevel 1 :AbilityLevel 1 :TargetAttachComponent (Variables|/Script/Engine.Character|GetMesh (Utilities|Casting|CastToFortPlayerPawnAthena (Variables|Default|GetCurrentLaserTarget))) :bReplicateLocationWhenUsingMinimalRepProxy false)) ; reach out to FX if  issue
)

(fn TriggerDelayedTargeting
  (bind _returnvalue (Utilities|Time|SetTimerbyEvent (EventDispatchers|CreateEvent :!event SnapshotLastSeenDirection) (ScalableFloat|GetValueatLevel (Variables|Default|GetTrackingDelay))))
  (Variables|Default|SetTargetDelayedHandle _returnvalue)
  (bind _returnvalue_2 (Utilities|Time|SetTimerbyEvent (EventDispatchers|CreateEvent :!event TargetCue) 0.7))
  (Variables|Default|SetTargetCueHandle _returnvalue_2))
